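# Physics asset for the back shoulder flaps on the M_MED_Orbit_Teal male medium player body, covering the left and right sides.
# Each flap bone has a large capsule body constrained to a tiny capsule on its MSTR bone, with equal swing and twist angle limits.
# Unconstrained capsules on both upper-arm twist bones supply arm shapes, apparently as colliders for the flaps.
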PhysicsAsset /BRCosmetics/Characters/Player/Male/Medium/Bodies/M_MED_Orbit_Teal/Meshes/M_MED_Orbit_Teal_Dyn_flaps
bodies {
 dyn_shoulder_flap_bk_l_01 [{
   shapeType=Capsule
   length=27.25
   radius=18.8591
   shapeName=dyn_shoulder_flap_bk_r_01_capsule
   center=<9.4565 1.7538 17.6223>
   rotation=<rot 84.3975 0 0>
  }]
 dyn_shoulder_flap_bk_l_MSTR [{shapeType=Capsule, length=0.2, radius=0.2, shapeName=dyn_shoulder_flap_bk_l_MSTR_capsule, rotation=<rot 90 0 0>}]
 dyn_shoulder_flap_bk_r_01 [{
   shapeType=Capsule
   length=27.25
   radius=18.8591
   shapeName=dyn_shoulder_flap_bk_r_01_capsule
   center=<9.4565 -1.7538 17.6223>
   rotation=<rot 84.3975 0 0>
  }]
 dyn_shoulder_flap_bk_r_MSTR [{shapeType=Capsule, length=0.2, radius=0.2, shapeName=dyn_shoulder_flap_bk_r_MSTR_capsule, rotation=<rot 90 0 0>}]
 upperarm_twist_01_l [{
   shapeType=Capsule
   length=31.25
   radius=6.25
   shapeName=upperarm_twist_01_r_capsule
   center=<14.6416 -3.6958 0>
   rotation=<rot 90 -7.125 -179.627>
  }]
 upperarm_twist_01_r [{shapeType=Capsule, length=31.25, radius=6.25, shapeName=upperarm_twist_01_r_capsule, center=<-14.6416 3.6958 0>, rotation=<rot 90 0 7.4981>}]
}
constraints [
 {
  bone1Name=dyn_shoulder_flap_bk_r_01
  bone2Name=dyn_shoulder_flap_bk_r_MSTR
  swing1LimitDegrees=45
  swing2LimitDegrees=45
  twistLimitDegrees=45
 }
 {
  bone1Name=dyn_shoulder_flap_bk_l_01
  bone2Name=dyn_shoulder_flap_bk_l_MSTR
  swing1LimitDegrees=45
  swing2LimitDegrees=45
  twistLimitDegrees=45
 }
]
