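# Physics asset for the dynamic holster on the M_MED_Orbit_Teal medium male player body cosmetic.
# The holster top body uses two angled capsules, one long and one short, while the top and bottom master bones each carry a tiny capsule.
# A single constraint ties the top to its master with tight swing limits and a looser twist limit.
# The bottom master body has no constraint.
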PhysicsAsset /BRCosmetics/Characters/Player/Male/Medium/Bodies/M_MED_Orbit_Teal/Meshes/M_MED_Orbit_Teal_Dyn_holster
bodies {
 dyn_holster_bott_MSTR [{shapeType=Capsule, length=0.505, radius=0.505, shapeName=dyn_holster_bott_MSTR_capsule}]
 dyn_holster_top [
  {
   shapeType=Capsule
   length=25.2945
   radius=3.3649
   shapeName=dyn_holster_top_capsule
   center=<17.0984 0 1.8503>
   rotation=<rot 85.9055 -0.0002 -0.0002>
  }
  {
   shapeType=Capsule
   length=4.8223
   radius=3.3649
   shapeName=dyn_holster_top_capsule
   center=<4.3865 5.5678 4.0067>
   rotation=<rot 78.6222 -48.6212 -160.3941>
  }
 ]
 dyn_holster_top_MSTR [{shapeType=Capsule, length=0.505, radius=0.505, shapeName=dyn_holster_top_MSTR_capsule}]
}
constraints [{bone1Name=dyn_holster_top, bone2Name=dyn_holster_top_MSTR, swing1LimitDegrees=2, swing2LimitDegrees=2, twistLimitDegrees=20}]
