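# Physics asset for the dynamic back loincloth on the male medium Orbit Teal player body.
# Center, left and right three-segment chains use limited swing and tight twist, hanging from master bodies with free angular joints to the pelvis; free-angular constraints also tie the side tips to the center tip.
# Pelvis, thigh and calf capsules represent the body, while oversized cloth capsules sit offset by their radius so their surface meets the bone.
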PhysicsAsset /BRCosmetics/Characters/Player/Male/Medium/Bodies/M_MED_Orbit_Teal/Meshes/M_MED_Orbit_Teal_Dyn_loincloth
bodies {
 calf_l [{
   shapeType=Capsule
   length=31.2098
   radius=28.2986
   shapeName=dyn_holster_top_MSTR_capsule
   center=<-24.4351 13.9936 0>
   rotation=<rot 90 0 0>
  }]
 calf_r [{
   shapeType=Capsule
   length=31.2098
   radius=28.2986
   shapeName=dyn_holster_top_MSTR_capsule
   center=<24.4351 -19.6569 0>
   rotation=<rot 90 0 180>
  }]
 dyn_loincloth_bk_01 [{shapeType=Capsule, length=19.25, radius=30, shapeName=dyn_holster_top_capsule, center=<9.7869 0 30>, rotation=<rot 90 0 0>}]
 dyn_loincloth_bk_02 [{shapeType=Capsule, length=19.25, radius=30, shapeName=dyn_holster_top_capsule, center=<9.7869 0 30>, rotation=<rot 90 0 0>}]
 dyn_loincloth_bk_03 [{shapeType=Capsule, length=19.25, radius=30, shapeName=dyn_holster_top_capsule, center=<9.7869 0 30>, rotation=<rot 90 0 0>}]
 dyn_loincloth_bk_MSTR [{shapeType=Capsule, length=0.505, radius=0.505, shapeName=dyn_holster_top_MSTR_capsule}]
 dyn_loincloth_bk_l_01 [{shapeType=Capsule, length=19.25, radius=30, shapeName=dyn_holster_top_capsule, center=<9.7869 0 30>, rotation=<rot 90 0 0>}]
 dyn_loincloth_bk_l_02 [{shapeType=Capsule, length=19.25, radius=30, shapeName=dyn_holster_top_capsule, center=<9.7869 0 30>, rotation=<rot 90 0 0>}]
 dyn_loincloth_bk_l_03 [{shapeType=Capsule, length=19.25, radius=30, shapeName=dyn_holster_top_capsule, center=<9.7869 0 30>, rotation=<rot 90 0 0>}]
 dyn_loincloth_bk_l_MSTR [{shapeType=Capsule, length=0.505, radius=0.505, shapeName=dyn_holster_top_MSTR_capsule}]
 dyn_roincloth_bk_r_01 [{shapeType=Capsule, length=19.25, radius=30, shapeName=dyn_holster_top_capsule, center=<9.7869 0 30>, rotation=<rot 90 0 0>}]
 dyn_roincloth_bk_r_02 [{shapeType=Capsule, length=19.25, radius=30, shapeName=dyn_holster_top_capsule, center=<9.7869 0 30>, rotation=<rot 90 0 0>}]
 dyn_roincloth_bk_r_03 [{shapeType=Capsule, length=19.25, radius=30, shapeName=dyn_holster_top_capsule, center=<9.7869 0 30>, rotation=<rot 90 0 0>}]
 dyn_roincloth_bk_r_MSTR [{shapeType=Capsule, length=0.505, radius=0.505, shapeName=dyn_holster_top_MSTR_capsule}]
 pelvis [{shapeType=Capsule, length=10.7744, radius=12.7471, shapeName=dyn_holster_top_MSTR_capsule, center=<-10.7828 0.2908 0>}]
 thigh_l [{
   shapeType=Capsule
   length=26.5655
   radius=10.0486
   shapeName=dyn_holster_top_MSTR_capsule
   center=<-20.4747 0.0011 0.0001>
   rotation=<rot 90 0 0>
  }]
 thigh_r [{
   shapeType=Capsule
   length=26.5655
   radius=10.0486
   shapeName=dyn_holster_top_MSTR_capsule
   center=<20.4747 -0.0011 -0.0001>
   rotation=<rot 90 0 180>
  }]
}
constraints [
 {bone1Name=dyn_loincloth_bk_l_01, bone2Name=dyn_loincloth_bk_l_MSTR, swing1LimitDegrees=45, swing2LimitDegrees=45, twistLimitDegrees=5}
 {bone1Name=dyn_loincloth_bk_l_02, bone2Name=dyn_loincloth_bk_l_01, swing1LimitDegrees=45, swing2LimitDegrees=45, twistLimitDegrees=5}
 {bone1Name=dyn_loincloth_bk_l_03, bone2Name=dyn_loincloth_bk_l_02, swing1LimitDegrees=45, swing2LimitDegrees=45, twistLimitDegrees=5}
 {bone1Name=dyn_loincloth_bk_01, bone2Name=dyn_loincloth_bk_MSTR, swing1LimitDegrees=45, swing2LimitDegrees=45, twistLimitDegrees=5}
 {bone1Name=dyn_loincloth_bk_02, bone2Name=dyn_loincloth_bk_01, swing1LimitDegrees=45, swing2LimitDegrees=45, twistLimitDegrees=5}
 {bone1Name=dyn_loincloth_bk_03, bone2Name=dyn_loincloth_bk_02, swing1LimitDegrees=45, swing2LimitDegrees=45, twistLimitDegrees=5}
 {bone1Name=dyn_roincloth_bk_r_01, bone2Name=dyn_roincloth_bk_r_MSTR, swing1LimitDegrees=45, swing2LimitDegrees=45, twistLimitDegrees=5}
 {bone1Name=dyn_roincloth_bk_r_02, bone2Name=dyn_roincloth_bk_r_01, swing1LimitDegrees=45, swing2LimitDegrees=45, twistLimitDegrees=5}
 {bone1Name=dyn_roincloth_bk_r_03, bone2Name=dyn_roincloth_bk_r_02, swing1LimitDegrees=45, swing2LimitDegrees=45, twistLimitDegrees=5}
 {
  bone1Name=dyn_loincloth_bk_l_MSTR
  bone2Name=pelvis
  swing1LimitDegrees=45
  swing1Motion=Free
  swing2LimitDegrees=45
  swing2Motion=Free
  twistLimitDegrees=45
  twistMotion=Free
 }
 {
  bone1Name=dyn_loincloth_bk_MSTR
  bone2Name=pelvis
  swing1LimitDegrees=45
  swing1Motion=Free
  swing2LimitDegrees=45
  swing2Motion=Free
  twistLimitDegrees=45
  twistMotion=Free
 }
 {
  bone1Name=dyn_roincloth_bk_r_MSTR
  bone2Name=pelvis
  swing1LimitDegrees=45
  swing1Motion=Free
  swing2LimitDegrees=45
  swing2Motion=Free
  twistLimitDegrees=45
  twistMotion=Free
 }
 {
  bone1Name=thigh_l
  bone2Name=pelvis
  swing1LimitDegrees=45
  swing1Motion=Free
  swing2LimitDegrees=45
  swing2Motion=Free
  twistLimitDegrees=45
  twistMotion=Free
 }
 {
  bone1Name=calf_r
  bone2Name=thigh_r
  swing1LimitDegrees=45
  swing1Motion=Free
  swing2LimitDegrees=45
  swing2Motion=Free
  twistLimitDegrees=45
  twistMotion=Free
 }
 {
  bone1Name=thigh_r
  bone2Name=pelvis
  swing1LimitDegrees=45
  swing1Motion=Free
  swing2LimitDegrees=45
  swing2Motion=Free
  twistLimitDegrees=45
  twistMotion=Free
 }
 {
  bone1Name=calf_l
  bone2Name=thigh_l
  swing1LimitDegrees=45
  swing1Motion=Free
  swing2LimitDegrees=45
  swing2Motion=Free
  twistLimitDegrees=45
  twistMotion=Free
 }
 {
  bone1Name=dyn_roincloth_bk_r_03
  bone2Name=dyn_loincloth_bk_03
  swing1LimitDegrees=45
  swing1Motion=Free
  swing2LimitDegrees=45
  swing2Motion=Free
  twistLimitDegrees=45
  twistMotion=Free
 }
 {
  bone1Name=dyn_loincloth_bk_l_03
  bone2Name=dyn_loincloth_bk_03
  swing1LimitDegrees=45
  swing1Motion=Free
  swing2LimitDegrees=45
  swing2Motion=Free
  twistLimitDegrees=45
  twistMotion=Free
 }
]
