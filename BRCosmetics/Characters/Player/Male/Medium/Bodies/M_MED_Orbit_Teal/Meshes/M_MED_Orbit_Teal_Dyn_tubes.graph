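# Physics asset for the dynamic tubes on the right arm of the M_MED_Orbit_Teal player body, covering four top and bottom tube segments.
# Each tube bone has a capsule body constrained to its own small MSTR anchor body, with tight, equal swing and twist limit angles.
# Unconstrained upper and lower right arm capsules supply body shapes for the arm.
# Shape names carry a dyn_holster_top prefix, unlike the tube bones.
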PhysicsAsset /BRCosmetics/Characters/Player/Male/Medium/Bodies/M_MED_Orbit_Teal/Meshes/M_MED_Orbit_Teal_Dyn_tubes
bodies {
 dyn_tubes_bott_r_01 [{shapeType=Capsule, length=5.7673, radius=3.3649, shapeName=dyn_holster_top_capsule, rotation=<rot 90 0 0>}]
 dyn_tubes_bott_r_02 [{shapeType=Capsule, length=5.7673, radius=3.3649, shapeName=dyn_holster_top_capsule, rotation=<rot 90 0 0>}]
 dyn_tubes_bott_r_02_MSTR [{shapeType=Capsule, length=0.505, radius=0.505, shapeName=dyn_holster_top_MSTR_capsule}]
 dyn_tubes_bott_r_MSTR [{shapeType=Capsule, length=0.505, radius=0.505, shapeName=dyn_holster_top_MSTR_capsule}]
 dyn_tubes_top_r_01 [{shapeType=Capsule, length=5.7673, radius=3.3649, shapeName=dyn_holster_top_capsule, rotation=<rot 90 0 0>}]
 dyn_tubes_top_r_02 [{shapeType=Capsule, length=5.7673, radius=3.3649, shapeName=dyn_holster_top_capsule, rotation=<rot 90 0 0>}]
 dyn_tubes_top_r_02_MSTR [{shapeType=Capsule, length=0.505, radius=0.505, shapeName=dyn_holster_top_MSTR_capsule}]
 dyn_tubes_top_r_MSTR [{shapeType=Capsule, length=0.505, radius=0.505, shapeName=dyn_holster_top_MSTR_capsule}]
 lowerarm_r [{
   shapeType=Capsule
   length=21.6047
   radius=6.7337
   shapeName=dyn_holster_top_MSTR_capsule
   center=<-13.753 -1.45 0.0001>
   rotation=<rot 90 0 0>
  }]
 upperarm_r [{
   shapeType=Capsule
   length=21.6047
   radius=6.7337
   shapeName=dyn_holster_top_MSTR_capsule
   center=<-13.8624 -0.773 2.4774>
   rotation=<rot 83.8546 -179.9999 -179.9999>
  }]
}
constraints [
 {bone1Name=dyn_tubes_top_r_01, bone2Name=dyn_tubes_top_r_MSTR, swing1LimitDegrees=5, swing2LimitDegrees=5, twistLimitDegrees=5}
 {bone1Name=dyn_tubes_bott_r_01, bone2Name=dyn_tubes_bott_r_MSTR, swing1LimitDegrees=5, swing2LimitDegrees=5, twistLimitDegrees=5}
 {bone1Name=dyn_tubes_top_r_02, bone2Name=dyn_tubes_top_r_02_MSTR, swing1LimitDegrees=5, swing2LimitDegrees=5, twistLimitDegrees=5}
 {bone1Name=dyn_tubes_bott_r_02, bone2Name=dyn_tubes_bott_r_02_MSTR, swing1LimitDegrees=5, swing2LimitDegrees=5, twistLimitDegrees=5}
]
