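# Post-process style AnimBP for the Orbit Teal cosmetic body, driving corrective deformation and secondary motion on gauntlets, tubes, flaps, holster and loincloth.
# A shared pose copies the parent mesh, then applies Control Rigs, bone constraints and rigid bodies.
# The loincloth uses cloth or rigid body, falling back to trail controllers with a DBNO variant.
# The event graph updates cloth, gravity, joint alphas and trail rotation from movement, skydiving and vehicle state.
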AnimBlueprint /BRCosmetics/Characters/Player/Male/Medium/Bodies/M_MED_Orbit_Teal/Meshes/M_MED_Orbit_Teal_AnimBP
target_skeleton=&/M_MED_Orbit_Teal_Skeleton
graphs [
 AnimGraph {nodes [
   n1 = Use cached pose 'trailCache'()
   n2 = Use cached pose 'dbnoCache'()
   n3 = Get bIsDBNO()
   Output Pose(Result=Blend Poses by bool(
     BlendPose_0=Use cached pose 'RbanCache'()
     BlendPose_1=Blend Poses by bool(BlendPose_0=@n2, BlendPose_1=@n1, bActiveValue=@n3) {blendType=Linear}
     bActiveValue=Get bRigidBodyEnabled()
    ) {blendType=Linear}) {layerGroup=None}
   n4 = Control Rig(Source=Copy Pose From Mesh() {bUseAttachedParent=true}) {
    alphaBoolBlend {blendOption=Linear}
    controlRigAssetReference {blueprintRigClass=&/M_MED_Orbit_Teal_Deform_CtrlRig_twists.M_MED_Orbit_Teal_Deform_CtrlRig_twists_C}
   }
   n5 = Control Rig!(
    Source=@n4
    Alpha
    KneeRot_fr_l_114deg:Vector
    KneeTrans_fr_l_114deg:Vector="0.258,-3.591,0.009"
    KneeScale_fr_l_114deg:Vector="1,1,1"
    KneeScale_fr_l_130deg:Vector="1,1,1"
    KneeRot_fr_l_130deg:Vector="0,0,-51.2"
    KneeTrans_fr_l_130deg:Vector="0.258,-3.591,0.009"
   ) -> (Pose) {
    alphaBoolBlend {blendOption=Linear}
    controlRigAssetReference {blueprintRigClass=&/M_MED_Orbit_Teal_Deform_CtrlRig_sculpters.M_MED_Orbit_Teal_Deform_CtrlRig_sculpters_C}
   }
   n6 = "Transform (Modify) Bone - Bone: dyn_gauntlet_l"(
    ComponentPose=Local To Component(LocalPose=@n5)
    Alpha=Get wrist_l_rot() -> (wrist_l_rot:double)
    Translation="-4.5,0,0"
    Rotation="0,0,0"
    Scale="1,1,1"
   ) {
    rotationSpace=BCS_BoneSpace
    translationMode=BMM_Additive
    translationSpace=BCS_BoneSpace
    alphaBoolBlend {blendOption=Linear}
    alphaScaleBiasClamp {bInterpResult=true, interpSpeedDecreasing=20, interpSpeedIncreasing=20}
    boneToModify {boneName=dyn_gauntlet_l}
    translation {x=-4.5}
   }
   n7 = "Transform (Modify) Bone - Bone: dyn_gauntlet_r"(ComponentPose=@n6, Alpha=Get wrist_r_rot() -> (wrist_r_rot:double), Translation="-4.5,0,0", Rotation="0,0,0", Scale="1,1,1") {
    rotationSpace=BCS_BoneSpace
    translationMode=BMM_Additive
    translationSpace=BCS_BoneSpace
    alphaBoolBlend {blendOption=Linear}
    alphaScaleBiasClamp {bInterpResult=true, interpSpeedDecreasing=20, interpSpeedIncreasing=20}
    boneToModify {boneName=dyn_gauntlet_r}
    translation {x=-4.5}
   }
   n8 = Get in vehicle scale alpha() -> (in vehicle scale alpha:double)
   n9 = "Transform (Modify) Bone - Bone: dyn_loincloth_bk_l_02"(ComponentPose=@n7, Alpha=@n8, Translation="0,0,0", Rotation="0,0,0", Scale="0.1,0.5,1") {scaleMode=BMM_Replace, alphaBoolBlend {blendOption=Linear}, boneToModify {boneName=dyn_loincloth_bk_l_02}, scale=<0.1 0.5>}
   n10 = "Transform (Modify) Bone - Bone: dyn_loincloth_bk_02"(ComponentPose=@n9, Alpha=@n8, Translation="0,0,0", Rotation="0,0,0", Scale="0.1,0.5,1") {scaleMode=BMM_Replace, alphaBoolBlend {blendOption=Linear}, boneToModify {boneName=dyn_loincloth_bk_02}, scale=<0.1 0.5>}
   n11 = "Transform (Modify) Bone - Bone: dyn_roincloth_bk_r_02"(ComponentPose=@n10, Alpha=@n8, Translation="0,0,0", Rotation="0,0,0", Scale="0.1,0.5,1") {scaleMode=BMM_Replace, alphaBoolBlend {blendOption=Linear}, boneToModify {boneName=dyn_roincloth_bk_r_02}, scale=<0.1 0.5>}
   n12 = "Constraint - dyn_tubes_top_r_MSTR (PPPPPPPP)"!(
    ComponentPose=@n11
    Alpha=1
    ConstraintWeights_0=0.3538
    ConstraintWeights_1:float=0.2792
    ConstraintWeights_2:float=0.1058
    ConstraintWeights_3:float=0.1032
    ConstraintWeights_4:float=0.0545
    ConstraintWeights_5:float=0.0411
    ConstraintWeights_6:float=0.0322
    ConstraintWeights_7:float=0.0302
   ) -> (Pose) {
    alphaBoolBlend {blendOption=Linear}
    boneToModify {boneName=dyn_tubes_top_r_MSTR}
    constraintSetup [
     {targetBone {boneName=upperarm_twist_01_r}}
     {targetBone {boneName=upperarm_twist_02_r}}
     {targetBone {boneName=clavicle_r}}
     {targetBone {boneName=deform_deltoid_fr_r}}
     {targetBone {boneName=spine_05}}
     {targetBone {boneName=deform_pectoral_fr_r}}
     {targetBone {boneName=upperarm_r}}
     {targetBone {boneName=spine_04}}
    ]
    constraintWeights [1,1,1,1,1,1,1,1]
   }
   n13 = "Constraint - dyn_tubes_bott_r_MSTR (PPPPPP)"!(
    ComponentPose=@n12
    Alpha=1
    ConstraintWeights_0=0.4886
    ConstraintWeights_1:float=0.2817
    ConstraintWeights_2:float=0.112
    ConstraintWeights_3:float=0.056
    ConstraintWeights_4:float=0.0342
    ConstraintWeights_5:float=0.0275
   ) -> (Pose) {
    alphaBoolBlend {blendOption=Linear}
    boneToModify {boneName=dyn_tubes_bott_r_MSTR}
    constraintSetup [
     {targetBone {boneName=upperarm_twist_02_r}}
     {targetBone {boneName=upperarm_twist_01_r}}
     {targetBone {boneName=clavicle_r}}
     {targetBone {boneName=upperarm_r}}
     {targetBone {boneName=deform_deltoid_fr_r}}
     {targetBone {boneName=spine_05}}
    ]
    constraintWeights [1,1,1,1,1,1]
   }
   n14 = "Constraint - dyn_tubes_bott_r_02_MSTR (PP)"!(ComponentPose=@n13, Alpha=1, ConstraintWeights_0=0.4, ConstraintWeights_1:float=0.6) -> (Pose) {
    alphaBoolBlend {blendOption=Linear}
    boneToModify {boneName=dyn_tubes_bott_r_02_MSTR}
    constraintSetup [{targetBone {boneName=lowerarm_twist_02_r}}, {targetBone {boneName=upperarm_r}}]
    constraintWeights [1,1]
   }
   n15 = "Constraint - dyn_tubes_top_r_02_MSTR (PP)"!(ComponentPose=@n14, Alpha=1, ConstraintWeights_0=0.4, ConstraintWeights_1:float=0.6) -> (Pose) {
    alphaBoolBlend {blendOption=Linear}
    boneToModify {boneName=dyn_tubes_top_r_02_MSTR}
    constraintSetup [{targetBone {boneName=lowerarm_twist_02_r}}, {targetBone {boneName=upperarm_r}}]
    constraintWeights [1,1]
   }
   n16 = Get elbow_rot_r() -> (elbow_rot_r:double)
   n17 = Weighted Look At - Bone: dyn_tubes_top_r_MSTR!(ComponentPose:Component Space Pose Link Structure=@n15, Alpha:float=@n16) -> (Pose:Component Space Pose Link Structure) {
    is=title FortAnimGraphNode_WeightedLookAt
    bUseLookUpAxis=true
    alphaBoolBlend {blendOption=Linear}
    boneToModify {boneName=dyn_tubes_top_r_MSTR}
    lookAtLocation {x=0}
    lookAtTarget {boneReference {boneName=dyn_tubes_top_r_02_MSTR}}
    lookAt_Axis {axis=<1 0>}
    lookUp_Axis {axis {x=0, z=1}}
   }
   n18 = Weighted Look At - Bone: dyn_tubes_bott_r_MSTR!(ComponentPose:Component Space Pose Link Structure=@n17, Alpha:float=1) -> (Pose:Component Space Pose Link Structure) {
    is=title FortAnimGraphNode_WeightedLookAt
    bUseLookUpAxis=true
    alphaBoolBlend {blendOption=Linear}
    boneToModify {boneName=dyn_tubes_bott_r_MSTR}
    lookAtLocation {x=0}
    lookAtTarget {boneReference {boneName=dyn_tubes_bott_r_02_MSTR}}
    lookAt_Axis {axis=<1 0>}
    lookUp_Axis {axis {x=0, z=1}}
   }
   n19 = Weighted Look At - Bone: dyn_tubes_top_r_02_MSTR!(ComponentPose:Component Space Pose Link Structure=@n18, Alpha:float=@n16) -> (Pose:Component Space Pose Link Structure) {
    is=title FortAnimGraphNode_WeightedLookAt
    bUseLookUpAxis=true
    alphaBoolBlend {blendOption=Linear}
    boneToModify {boneName=dyn_tubes_top_r_02_MSTR}
    lookAtLocation {x=0}
    lookAtTarget {boneReference {boneName=dyn_tubes_top_r_MSTR}}
    lookAt_Axis {axis=<1 0>}
    lookUp_Axis {axis {x=0, z=1}}
   }
   n20 = Weighted Look At - Bone: dyn_tubes_bott_r_02_MSTR!(ComponentPose:Component Space Pose Link Structure=@n19, Alpha:float=1) -> (Pose:Component Space Pose Link Structure) {
    is=title FortAnimGraphNode_WeightedLookAt
    bUseLookUpAxis=true
    alphaBoolBlend {blendOption=Linear}
    boneToModify {boneName=dyn_tubes_bott_r_02_MSTR}
    lookAtLocation {x=0}
    lookAtTarget {boneReference {boneName=dyn_tubes_bott_r_MSTR}}
    lookAt_Axis {axis=<1 0>}
    lookUp_Axis {axis {x=0, z=1}}
   }
   n21 = RigidBody!(ComponentPose=@n20, Alpha, OverrideWorldGravity:Vector="0,0,-980", ExternalForce) -> (Pose) {
    bClampLinearTranslationLimitToRefPose=true
    bOverrideWorldGravity=true
    overridePhysicsAsset=&/M_MED_Orbit_Teal_Dyn_tubes
    simulationSpace=BaseBoneSpace
    alphaBoolBlend {blendOption=Linear}
    baseBoneRef {boneName=upperarm_twist_01_r}
    componentAppliedLinearAccClamp=<3000 3000 3000>
    overrideWorldGravity {z=-980}
    simSpaceSettings {
     maxAngularAcceleration=1000
     maxAngularVelocity=3
     maxLinearAcceleration=250
     maxLinearVelocity=500
     velocityScaleZ=0
     worldAlpha=1
     externalLinearDragV {y=0.5}
    }
   }
   n22 = RigidBody(ComponentPose=@n21, OverrideWorldGravity="0,0,-980") {
    bClampLinearTranslationLimitToRefPose=true
    bOverrideWorldGravity=true
    overridePhysicsAsset=&/M_MED_Orbit_Teal_Dyn_flaps
    simulationSpace=BaseBoneSpace
    alphaBoolBlend {blendOption=Linear}
    baseBoneRef {boneName=spine_05}
    componentAppliedLinearAccClamp=<3000 3000 3000>
    overrideWorldGravity {z=-980}
    simSpaceSettings {
     maxAngularAcceleration=1000
     maxAngularVelocity=3
     maxLinearAcceleration=250
     maxLinearVelocity=500
     velocityScaleZ=0
     externalLinearDragV {y=0.5}
    }
   }
   n23 = "Constraint - dyn_holster_bott_MSTR (PPPPPPPP)"!(
    ComponentPose=@n22
    Alpha=1
    ConstraintWeights_0=0.5357
    ConstraintWeights_1:float=0.3307
    ConstraintWeights_2:float=0.0468
    ConstraintWeights_3:float=0.041
    ConstraintWeights_4:float=0.0196
    ConstraintWeights_5:float=0.014
    ConstraintWeights_6:float=0.0112
    ConstraintWeights_7:float=0.001
   ) -> (Pose) {
    alphaBoolBlend {blendOption=Linear}
    boneToModify {boneName=dyn_holster_bott_MSTR}
    constraintSetup [
     {targetBone {boneName=thigh_r}}
     {targetBone {boneName=thigh_twist_01_r}}
     {targetBone {boneName=deform_knee_fr_01_r}}
     {targetBone {boneName=deform_knee_bk_01_r}}
     {targetBone {boneName=pelvis}}
     {targetBone {boneName=deform_thigh_fr_r}}
     {targetBone {boneName=calf_r}}
     {targetBone {boneName=deform_thigh_bk_r}}
    ]
    constraintWeights [1,1,1,1,1,1,1,1]
   }
   n24 = RigidBody(ComponentPose=@n23, OverrideWorldGravity="0,0,-980") {
    bClampLinearTranslationLimitToRefPose=true
    bOverrideWorldGravity=true
    overridePhysicsAsset=&/M_MED_Orbit_Teal_Dyn_holster
    simulationSpace=BaseBoneSpace
    alphaBoolBlend {blendOption=Linear}
    baseBoneRef {boneName=thigh_twist_01_r}
    componentAppliedLinearAccClamp=<3000 3000 3000>
    overrideWorldGravity {z=-980}
    simSpaceSettings {
     maxAngularAcceleration=1000
     maxAngularVelocity=3
     maxLinearAcceleration=250
     maxLinearVelocity=500
     velocityScaleZ=0
     worldAlpha=1
     externalLinearDragV {y=0.5}
    }
   }
   Save cached pose 'StartupCache'(Pose=Component To Local(ComponentPose=@n24)) {cacheName=StartupCache}
   n25 = Get TrailControllerRot() -> (TrailControllerRot:Rotator)
   n26 = "Transform (Modify) Bone - Bone: dyn_loincloth_bk_l_MSTR"(ComponentPose=Local To Component(LocalPose=Use cached pose 'StartupCache'()), Alpha=1, Translation="0,0,0", Rotation=@n25, Scale="1,1,1") {rotationMode=BMM_Additive, rotationSpace=BCS_BoneSpace, alphaBoolBlend {blendOption=Linear}, boneToModify {boneName=dyn_loincloth_bk_l_MSTR}}
   n27 = "Transform (Modify) Bone - Bone: dyn_loincloth_bk_MSTR"(ComponentPose=@n26, Alpha=1, Translation="0,0,0", Rotation=@n25, Scale="1,1,1") {rotationMode=BMM_Additive, rotationSpace=BCS_BoneSpace, alphaBoolBlend {blendOption=Linear}, boneToModify {boneName=dyn_loincloth_bk_MSTR}}
   n28 = "Transform (Modify) Bone - Bone: dyn_roincloth_bk_r_MSTR"(ComponentPose=@n27, Alpha=1, Translation="0,0,0", Rotation=@n25, Scale="1,1,1") {rotationMode=BMM_Additive, rotationSpace=BCS_BoneSpace, alphaBoolBlend {blendOption=Linear}, boneToModify {boneName=dyn_roincloth_bk_r_MSTR}}
   n29 = Get BwdMovement() -> (BwdMovement:double)
   n30 = Trail controller - Bone: dyn_loincloth_bk_l_03(ComponentPose=@n28, Alpha=@n29) {
    bUsePlanarLimit=true
    chainLength=3
    debugLifeTime=0.1
    maxDeltaTime=0.333
    alphaScaleBiasClamp {bInterpResult=true}
    rotationLimits [{limitMax=<180 180 180>}, {limitMax=<180 180 180>}, {limitMax=<180 180 180>}]
    rotationOffsets [<0 0 0>, <0 0 0>, <0 0 0>]
    trailBone {boneName=dyn_loincloth_bk_l_03}
    trailRelaxationSpeed {editorCurveData {keys [<at -0.0134 19.3841 RCIM_Linear>,<at 0.5079 12.8986 RCIM_Linear>,<at 0.9955 9.6044 RCIM_Linear>]}}
   }
   n31 = Trail controller - Bone: dyn_loincloth_bk_03(ComponentPose=@n30, Alpha=@n29) {
    bUsePlanarLimit=true
    chainLength=3
    debugLifeTime=0.1
    maxDeltaTime=0.0333
    alphaScaleBiasClamp {bInterpResult=true}
    rotationLimits [{limitMax=<180 180 180>}, {limitMax=<180 180 180>}, {limitMax=<180 90 180>, limitMin=<-180 -90 -180>}]
    rotationOffsets [<0 0 0>, <0 0 0>, <0 0 0>]
    trailBone {boneName=dyn_loincloth_bk_03}
    trailRelaxationSpeed {editorCurveData {keys [<at 0 15 RCIM_Linear>,<at 0.5289 13.1838 RCIM_Linear>,<at 1 6 RCIM_Linear>]}}
   }
   n32 = Trail controller - Bone: dyn_roincloth_bk_r_03(ComponentPose=@n31, Alpha=@n29) {
    bUsePlanarLimit=true
    chainLength=3
    debugLifeTime=0.1
    maxDeltaTime=0.0333
    alphaScaleBiasClamp {bInterpResult=true}
    rotationLimits [{limitMax=<180 180 180>}, {limitMax=<180 180 180>}, {limitMax=<180 90 180>, limitMin=<-180 -90 -180>}]
    rotationOffsets [<0 0 0>, <0 0 0>, <0 0 0>]
    trailBone {boneName=dyn_roincloth_bk_r_03}
    trailRelaxationSpeed {editorCurveData {keys [<at 0 15 RCIM_Linear>,<at 0.5289 13.1838 RCIM_Linear>,<at 1 6 RCIM_Linear>]}}
   }
   Save cached pose 'trailCache'(Pose=Component To Local(ComponentPose=@n32)) {cacheName=trailCache}
   n33 = "Transform (Modify) Bone - Bone: dyn_loincloth_bk_l_01"(ComponentPose=Local To Component(LocalPose=Use cached pose 'StartupCache'()), Alpha=1, Translation="0,0,0", Rotation="5,0,0", Scale="1,1,1") {
    rotationMode=BMM_Additive
    rotationSpace=BCS_BoneSpace
    alphaBoolBlend {blendOption=Linear}
    boneToModify {boneName=dyn_loincloth_bk_l_01}
    rotation {pitch=5}
   }
   n34 = "Transform (Modify) Bone - Bone: dyn_loincloth_bk_l_02"(ComponentPose=@n33, Alpha=1, Translation="0,0,0", Rotation="-10,0,0", Scale="1,1,1") {
    rotationMode=BMM_Additive
    rotationSpace=BCS_BoneSpace
    alphaBoolBlend {blendOption=Linear}
    boneToModify {boneName=dyn_loincloth_bk_l_02}
    rotation {pitch=-10}
   }
   n35 = "Transform (Modify) Bone - Bone: dyn_loincloth_bk_l_03"(ComponentPose=@n34, Alpha=1, Translation="0,0,0", Rotation="-25,0,0", Scale="1,1,1") {
    rotationMode=BMM_Additive
    rotationSpace=BCS_BoneSpace
    alphaBoolBlend {blendOption=Linear}
    boneToModify {boneName=dyn_loincloth_bk_l_03}
    rotation {pitch=-25}
   }
   n36 = "Transform (Modify) Bone - Bone: dyn_loincloth_bk_01"(ComponentPose=@n35, Alpha=1, Translation="0,0,0", Rotation="5,0,0", Scale="1,1,1") {
    rotationMode=BMM_Additive
    rotationSpace=BCS_BoneSpace
    alphaBoolBlend {blendOption=Linear}
    boneToModify {boneName=dyn_loincloth_bk_01}
    rotation {pitch=5}
   }
   n37 = "Transform (Modify) Bone - Bone: dyn_loincloth_bk_02"(ComponentPose=@n36, Alpha=1, Translation="0,0,0", Rotation="-10,0,0", Scale="1,1,1") {
    rotationMode=BMM_Additive
    rotationSpace=BCS_BoneSpace
    alphaBoolBlend {blendOption=Linear}
    boneToModify {boneName=dyn_loincloth_bk_02}
    rotation {pitch=-10}
   }
   n38 = "Transform (Modify) Bone - Bone: dyn_loincloth_bk_03"(ComponentPose=@n37, Alpha=1, Translation="0,0,0", Rotation="-25,0,0", Scale="1,1,1") {
    rotationMode=BMM_Additive
    rotationSpace=BCS_BoneSpace
    alphaBoolBlend {blendOption=Linear}
    boneToModify {boneName=dyn_loincloth_bk_03}
    rotation {pitch=-25}
   }
   n39 = "Transform (Modify) Bone - Bone: dyn_roincloth_bk_r_01"(ComponentPose=@n38, Alpha=1, Translation="0,0,0", Rotation="5,0,0", Scale="1,1,1") {
    rotationMode=BMM_Additive
    rotationSpace=BCS_BoneSpace
    alphaBoolBlend {blendOption=Linear}
    boneToModify {boneName=dyn_roincloth_bk_r_01}
    rotation {pitch=5}
   }
   n40 = "Transform (Modify) Bone - Bone: dyn_roincloth_bk_r_02"(ComponentPose=@n39, Alpha=1, Translation="0,0,0", Rotation="-10,0,0", Scale="1,1,1") {
    rotationMode=BMM_Additive
    rotationSpace=BCS_BoneSpace
    alphaBoolBlend {blendOption=Linear}
    boneToModify {boneName=dyn_roincloth_bk_r_02}
    rotation {pitch=-10}
   }
   n41 = "Transform (Modify) Bone - Bone: dyn_roincloth_bk_r_03"(ComponentPose=@n40, Alpha=1, Translation="0,0,0", Rotation="-25,0,0", Scale="1,1,1") {
    rotationMode=BMM_Additive
    rotationSpace=BCS_BoneSpace
    alphaBoolBlend {blendOption=Linear}
    boneToModify {boneName=dyn_roincloth_bk_r_03}
    rotation {pitch=-25}
   }
   n42 = Trail controller - Bone: dyn_loincloth_bk_l_03!(ComponentPose=@n41, Alpha=1, RelaxationSpeedScale:float=2, FakeVelocity:Vector="0,0,0") -> (Pose) {
    bShowTrailLocation=true
    bUsePlanarLimit=true
    chainLength=3
    debugLifeTime=0.1
    maxDeltaTime=0.0333
    relaxationSpeedScale=1.2
    alphaScaleBiasClamp {bInterpResult=true, interpSpeedDecreasing=25, interpSpeedIncreasing=25}
    rotationLimits [{limitMax=<180 180 180>}, {limitMax=<180 180 180>}, {limitMax=<180 180 180>}]
    rotationOffsets [<0 0 0>, <0 0 0>, <0 0 0>]
    trailBone {boneName=dyn_loincloth_bk_l_03}
    trailRelaxationSpeed {editorCurveData {keys [<at 0 15 RCIM_Linear>,<at 1 6 RCIM_Linear>]}}
   }
   n43 = Trail controller - Bone: dyn_loincloth_bk_03!(ComponentPose=@n42, Alpha=1, RelaxationSpeedScale:float=2, FakeVelocity:Vector="0,0,0") -> (Pose) {
    bShowTrailLocation=true
    bUsePlanarLimit=true
    chainLength=3
    debugLifeTime=0.1
    maxDeltaTime=0.0333
    relaxationSpeedScale=1.2
    alphaScaleBiasClamp {bInterpResult=true, interpSpeedDecreasing=25, interpSpeedIncreasing=25}
    rotationLimits [{limitMax=<180 180 180>}, {limitMax=<180 180 180>}, {limitMax=<180 180 180>}]
    rotationOffsets [<0 0 0>, <0 0 0>, <0 0 0>]
    trailBone {boneName=dyn_loincloth_bk_03}
    trailRelaxationSpeed {editorCurveData {keys [<at 0 15 RCIM_Linear>,<at 1 6 RCIM_Linear>]}}
   }
   n44 = Trail controller - Bone: dyn_roincloth_bk_r_03!(ComponentPose=@n43, Alpha=1, RelaxationSpeedScale:float=2, FakeVelocity:Vector="0,0,0") -> (Pose) {
    bShowTrailLocation=true
    bUsePlanarLimit=true
    chainLength=3
    debugLifeTime=0.1
    maxDeltaTime=0.0333
    relaxationSpeedScale=1.2
    alphaScaleBiasClamp {bInterpResult=true, interpSpeedDecreasing=25, interpSpeedIncreasing=25}
    rotationLimits [{limitMax=<180 180 180>}, {limitMax=<180 180 180>}, {limitMax=<180 180 180>}]
    rotationOffsets [<0 0 0>, <0 0 0>, <0 0 0>]
    trailBone {boneName=dyn_roincloth_bk_r_03}
    trailRelaxationSpeed {editorCurveData {keys [<at 0 15 RCIM_Linear>,<at 1 6 RCIM_Linear>]}}
   }
   Save cached pose 'dbnoCache'(Pose=Component To Local(ComponentPose=@n44)) {cacheName=dbnoCache}
   n45 = RigidBody(
    ComponentPose=Local To Component(LocalPose=Use cached pose 'StartupCache'())
    Alpha=Get in vehicle rigid alpha() -> (in vehicle rigid alpha:double)
    OverrideWorldGravity=Get GravityOverride()
   ) {
    bClampLinearTranslationLimitToRefPose=true
    bOverrideWorldGravity=true
    overridePhysicsAsset=&/M_MED_Orbit_Teal_Dyn_loincloth
    simulationSpace=BaseBoneSpace
    alphaBoolBlend {blendOption=Linear}
    baseBoneRef {boneName=pelvis}
    componentAppliedLinearAccClamp=<3000 3000 3000>
    componentLinearAccScale=<0.5 0.3 0.3>
    componentLinearVelScale=<0.5 0.3 0.3>
    overrideWorldGravity {z=-980}
    simSpaceSettings {
     maxAngularAcceleration=1000
     maxAngularVelocity=3
     maxLinearAcceleration=700
     maxLinearVelocity=500
     velocityScaleZ=0
     worldAlpha=1
     externalLinearDragV=<0.2 0.2>
    }
   }
   Save cached pose 'RbanCache'(Pose=Blend Poses by bool(
     BlendPose_0=Use cached pose 'StartupCache'()
     BlendPose_1=Component To Local(ComponentPose=@n45)
     BlendTime_0=0.2
     BlendTime_1=0.5
     bActiveValue=Get bClothEnabled()
    ) {blendType=Linear}) {cacheName=RbanCache}
  ]}
 EventGraph {nodes [
   n1 = Event BlueprintUpdateAnimation()
   > Set DeltaTimeLocal(DeltaTimeLocal:double=@n1.DeltaTimeX) -> (Output_Get:double)
   > n2 = Sequence!(execute) -> (then_0, then_1, then_2:Exec, then_3:Exec, then_4:Exec, then_5:Exec)
   > n3 = Branch(Condition=Get bClothEnabled())
   > n4 = "Cloth Controller Function (Parameter Asset) (Nativized) DEPRECATED"!(
    execute
    CharacterPartAnimInstance:Custom Character Part Anim Instance Object Reference=Self-Reference()
    ClothingAssetName:Name=M_MED_Orbit_Teal_Clothing_0
    ClothCoreSettings:"Cloth Core Settings Structure (by ref)"=Get Cloth Core Settings Native() -> (Cloth Core Settings Native:Cloth Core Settings Structure)
    WindGustLocalSettings:"Wind Gust Settings Structure (by ref)"=Get Wind Gust Local Settings Native() -> (Wind Gust Local Settings Native:Wind Gust Settings Structure)
    ClothUpAxis:Vector="1,0,0"
    OnGroundMovingSpeedLimit:float=2
    OnGroundMovingAngularVelMax:float=0.05
    OnGroundMovingFictitiousVelMax:float=0.6
    LinVelScaleFwdBwdClamp:float=0
    LinVelScaleUpDwnClamp:float=0.7
    DeltaTime:float=Get DeltaTimeLocal() -> (DeltaTimeLocal:double)
    GameTimeInSeconds:float=GetGameTimeInSeconds()
    StateLogicLayerName:Name=none
    Parameters:Fort Character Dynamics Parameters Object Reference=&/M_MED_Orbit_Teal_Parameters
   ) -> (
    then
    OutCurrentState:Name=None
    OutClothParameters:Cloth Parameters Structure
    OutClothCoreSettings:Cloth Core Settings Structure
    OutWindGustSettings:Wind Gust Parameters Structure
    OutWindGustLocalSettings:Wind Gust Settings Structure
    OutWindVelocity:Vector
    OutLinearVelVector:Vector
   )
   > Set Wind Gust Local Settings Native Out(Wind Gust Local Settings Native Out:Wind Gust Settings Structure=@n4.OutWindGustLocalSettings) -> (Output_Get:Wind Gust Settings Structure)
   n5 = "Gravity Override State Function (Parameter Asset) (Nativized)"(
    execute=@n3.else
    CharacterPartAnimInstance=Self-Reference()
    GameTimeInSeconds=GetGameTimeInSeconds()
    Parameters=&M_MED_Orbit_Teal_Parameters
   )
   > Set GravityOverride(GravityOverride=@n5)
   n6 = GetPartSkeletalMeshComponent()
   Set wrist_l_rot(
    execute=@n2.then_1
    wrist_l_rot:double=Get Distance Between Two Sockets!(
     Component:Skeletal Mesh Component Object Reference=@n6
     SocketOrBoneNameA:Name=dyn_hand_loc_tracker_l
     SocketSpaceA:ERelativeTransformSpace Enum=RTS_World
     SocketOrBoneNameB:Name=index_01_l
     SocketSpaceB:ERelativeTransformSpace Enum=RTS_World
     bRemapRange:bool=true
     InRangeMin:float=-5
     InRangeMax:float=9.38
     OutRangeMin:float=1
     OutRangeMax:float=0
    ) -> (ReturnValue:float) {is=title K2Node_CallFunction}
   ) -> (Output_Get:double)
   > Set wrist_r_rot(wrist_r_rot:double=Get Distance Between Two Sockets(
     Component=@n6
     SocketOrBoneNameA=dyn_hand_loc_tracker_r
     SocketSpaceA=RTS_World
     SocketOrBoneNameB=index_01_r
     SocketSpaceB=RTS_World
     bRemapRange=true
     InRangeMin=0
     InRangeMax=9.38
     OutRangeMin=1
     OutRangeMax=0
    )) -> (Output_Get:double)
   Set elbow_rot_r(
    execute=@n2.then_2
    elbow_rot_r:double=GetAlphaFromReferenceTransformDelta(SkeletalMeshComponent=@n6, BoneName=lowerarm_r, BaseName=upperarm_r, InputRange=(X=0,Y=80), OutputRange=(X=1,Y=0), TwistAxis="1,0,0")
   ) -> (Output_Get:double)
   > Set upperarm_rotZ_r(upperarm_rotZ_r:double=GetAlphaFromReferenceTransformDelta(SkeletalMeshComponent=@n6, BoneName=upperarm_r, BaseName=clavicle_r, DeltaComponentType=EulerY, InputRange=(X=-20,Y=20), TwistAxis="0,0,1")) -> (Output_Get:double)
   > Set upperarm_rotZ_l(upperarm_rotZ_l:double=GetAlphaFromReferenceTransformDelta(SkeletalMeshComponent=@n6, BoneName=upperarm_l, BaseName=clavicle_l, DeltaComponentType=EulerY, InputRange=(X=-5,Y=20), TwistAxis="0,0,1")) -> (Output_Get:double)
   n7 = PlayerMovement(execute=@n2.then_3, AnimInstance=Self-Reference(), XDotProductCoefficient=1, YDotProductCoefficient=1, ZDotProductCoefficient=1)
   n8 = MapRangeClamped(Value=@n7.YMovement, InRangeA=0, InRangeB=1)
   n9 = "Absolute (Float)"(A=@n7)
   > Set BwdMovement(BwdMovement:double=MapRangeClamped(Value=float + float(A=@n9, B=@n8) {is=title K2Node_CommutativeAssociativeBinaryOperator}, InRangeA=0, InRangeB=1.4)) -> (Output_Get:double)
   n10 = Branch(execute=@n2.then_4, Condition=Get bIsSwimSprinting() -> (bIsSwimSprinting:bool))
   > Set TrailControllerRot(TrailControllerRot:Rotator="15,0, 0") -> (Output_Get:Rotator)
   n11 = Branch(execute=@n10.else, Condition=Get bIsSkyDiving())
   > n12 = Branch(Condition=Get bIsParachuteOpened())
   > Set TrailControllerRot(TrailControllerRot="45,0, 0")
   n13 = Branch(execute=@n12.else, Condition="float > float"(A=Get PawnVelocityZ(), B=50) {is=title K2Node_CallFunction})
   > Set BwdMovement(BwdMovement=float * float(A=Get BwdMovement(), B=0))
   > Set TrailControllerRot(TrailControllerRot="45,0, 0")
   n14 = Branch(execute=@n13.else, Condition=Get bIsSkydiveDiveMode())
   > Set TrailControllerRot(TrailControllerRot="15,0, 0")
   Set TrailControllerRot(execute=@n14.else, TrailControllerRot="60,0, 0")
   Set TrailControllerRot(execute=@n11.else, TrailControllerRot="15,0, 0")
   > Set BwdMovement(BwdMovement=float * float(A=Get BwdMovement(), B=1))
   n15 = Branch(execute=@n2.then_5, Condition=Get bIsInVehicle() -> (bIsInVehicle:bool))
   n16 = Get bVehicleSeatForcesCrouch() -> (bVehicleSeatForcesCrouch:bool)
   n17 = Get bIsGolfCartVehicleBackPassenger() -> (bIsGolfCartVehicleBackPassenger:bool)
   n18 = Get bIsJackalVehicleDriver() -> (bIsJackalVehicleDriver:bool)
   n19 = Get bIsInShoppingCart() -> (bIsInShoppingCart:bool)
   n20 = Get bIsInCannon() -> (bIsInCannon:bool)
   n21 = Get bIsMeatballPassenger() -> (bIsMeatballPassenger:bool)
   > n22 = Branch(Condition=NOT Boolean(A=OR Boolean!(A=@n16, B=@n17, C:bool=@n18, D:bool=@n19, E:bool=@n20, F:bool=@n21) -> (ReturnValue)))
   > Set in vehicle rigid alpha(in vehicle rigid alpha:double=0) -> (Output_Get:double)
   > Set in vehicle scale alpha(in vehicle scale alpha:double=1) -> (Output_Get:double)
   n23 = GetPartSkeletalMeshComponent()
   > Set ClothBlendWeight!(execute, ClothBlendWeight:float=0, self:Skeletal Mesh Component Object Reference=@n23) -> (then, Output_Get:float)
   Set in vehicle rigid alpha(execute=[@n22.else, @n15.else], in vehicle rigid alpha=1)
   > Set in vehicle scale alpha(in vehicle scale alpha=0)
   > Set ClothBlendWeight(ClothBlendWeight=1, self=@n23)
   Event OnBeginSkydiving()
   > Set TrailControllerRot(TrailControllerRot="-65,0,0")
   Event OnEndSkydiving()
   > Set TrailControllerRot(TrailControllerRot="-10,0,0")
   Break Cloth Setting Struct!(ClothSettingStruct:"Cloth Setting Struct Structure (by ref)") -> (
    LinearVelocityScale_7_#1:Vector=~
    AngularVelocityScale_6_#2:float=~
    FictitiousAngularScale_10_#3:float=~
    GravityOverride_15_#4:Vector=~
    DragCoefficient_23_#5:Vector 2D Structure
    LiftCoefficient_24_#6:Vector 2D Structure
    JointName_27_#7:Name
   ) {is=title K2Node_BreakStruct}
   DrawVectorFromJointFunction!(
    execute
    SkelMeshComponent:Scene Component Object Reference=GetPartSkeletalMeshComponent()
    JointName:Name=pelvis
    LineEnd:Vector="0, 0, 0"
    JointOffset:Vector="0, 0, 0"
    LineColor:Linear Color Structure=(R=1,G=0,B=0,A=1)
    Thickness:double=2
    Foreground:EDrawDebugSceneDepthPriorityGroup Enum=Foreground
   ) -> (then)
  ]}
]
variables [
 BwdMovement {type=number}
 Cloth Core Settings Native {
  type=ClothCoreSettings
  default {
   collisionThickness=0
   dampingCoefficient=0.1
   frictionCoefficient=0.3
   localDampingCoefficient=0.1
   numberIterations=2
   numberSubsteps=1
   selfCollisionThickness=0
   animDriveDamping=<0 1>
   animDriveStiffness=<0 1>
   areaStiffness=<1 1>
   bendingStiffness=<1 1>
   edgeStiffness=<1 1>
   tetherStiffness=<1 1>
  }
 }
 DeltaTimeLocal {type=number}
 elbow_rot_r {type=number}
 GravityOverride {type=Vector}
 in vehicle rigid alpha {type=number}
 in vehicle scale alpha {type=number}
 TrailControllerRot {type=Rotator}
 upperarm_rotZ_l {type=number}
 upperarm_rotZ_r {type=number}
 Wind Gust Local Settings Native {type=WindGustSettings, default {elapsedTime=0, finalGustStrengthLocal=0, gustDuration=0, gustStrength=0, timeBetweenGusts=0}}
 Wind Gust Local Settings Native Out {type=WindGustSettings, default {elapsedTime=0, finalGustStrengthLocal=0, gustDuration=0, gustStrength=0, timeBetweenGusts=0}}
 wrist_l_rot {type=number}
 wrist_r_rot {type=number}
]
