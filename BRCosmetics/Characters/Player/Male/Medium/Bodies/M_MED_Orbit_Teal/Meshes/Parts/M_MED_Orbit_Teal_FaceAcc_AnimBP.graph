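# Face-accessory AnimBP for the Orbit Teal character, posing its head-mounted rangefinder over the parent mesh's copied pose.
# Targeting rotates dyn_rangefinder and head_alt through interpolated bone modifiers, alongside a head-driven neck fix.
# HeadLimits, set by the Orbit Teal emote or DisableHeadLimits, switches to head rotation matching with the visor emote layered on the rangefinder.
# The event graph also raises material emissive, after a delay, once the rangefinder rotates past a threshold.
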AnimBlueprint /BRCosmetics/Characters/Player/Male/Medium/Bodies/M_MED_Orbit_Teal/Meshes/Parts/M_MED_Orbit_Teal_FaceAcc_AnimBP
target_skeleton=&/M_MED_Orbit_Teal_FaceAcc_Skeleton
graphs [
 AnimGraph {nodes [
   n1 = "Transform (Modify) Bone - Bone: head_alt"(
    ComponentPose=Local To Component(LocalPose=Use cached pose 'preEmoteCache'())
    Alpha=0
    Translation="0,0,0"
    Rotation=Get head rot() -> (head rot:Rotator)
    Scale="1,1,1"
   ) {
    rotationMode=BMM_Replace
    rotationSpace=BCS_BoneSpace
    alphaBoolBlend {blendOption=Linear}
    alphaScaleBiasClamp {bInterpResult=true, interpSpeedDecreasing=5, interpSpeedIncreasing=20}
    boneToModify {boneName=head_alt}
    rotation {yaw=-10}
   }
   n2 = Get isTargetingFloat() -> (isTargetingFloat:double)
   n3 = "Transform (Modify) Bone - Bone: dyn_rangefinder"(ComponentPose=@n1, Alpha=@n2, Translation="0,0,0", Rotation="90,0.000,0.000", Scale="1,1,1") {
    rotationMode=BMM_Additive
    rotationSpace=BCS_BoneSpace
    alphaBoolBlend {blendOption=Linear}
    alphaScaleBiasClamp {bInterpResult=true, interpSpeedDecreasing=9, interpSpeedIncreasing=20}
    boneToModify {boneName=dyn_rangefinder}
    rotation {pitch=90}
   }
   n4 = "Transform (Modify) Bone - Bone: head_alt"(ComponentPose=@n3, Alpha=@n2, Translation="0,0,0", Rotation="0,-15,0.000", Scale="1,1,1") {
    rotationMode=BMM_Additive
    rotationSpace=BCS_BoneSpace
    alphaBoolBlend {blendOption=Linear}
    alphaScaleBiasClamp {bInterpResult=true, interpSpeedDecreasing=5, interpSpeedIncreasing=20}
    boneToModify {boneName=head_alt}
    rotation {yaw=-15}
   }
   n5 = "Transform (Modify) Bone - Bone: dyn_neck_fix"(ComponentPose=@n4, Alpha=Get head() -> (head:double), Translation="-3,-2,0", Rotation="0,0,0", Scale="1,0.5,0.1") {
    rotationMode=BMM_Additive
    rotationSpace=BCS_BoneSpace
    scaleMode=BMM_Additive
    scaleSpace=BCS_BoneSpace
    translationMode=BMM_Additive
    translationSpace=BCS_BoneSpace
    alphaBoolBlend {blendOption=Linear}
    boneToModify {boneName=dyn_neck_fix}
    scale {y=0.5, z=0.1}
    translation=<-3 -2>
   }
   n6 = "Constraint - head_alt (P)"(ComponentPose=@n5, Alpha=1, ConstraintWeights_0=1) {
    alphaBoolBlend {blendOption=Linear}
    alphaScaleBiasClamp {bInterpResult=true}
    boneToModify {boneName=head_alt}
    constraintSetup [{targetBone {boneName=head}}]
    constraintWeights [1]
   }
   n7 = "Transform (Modify) Bone - Bone: head_alt"!(ComponentPose=@n6, Alpha=1, Rotation="0,-12,0") -> (Pose) {
    rotationMode=BMM_Replace
    rotationSpace=BCS_BoneSpace
    alphaBoolBlend {blendOption=Linear}
    alphaScaleBiasClamp {bInterpResult=true, interpSpeedDecreasing=1, interpSpeedIncreasing=1}
    boneToModify {boneName=head_alt}
    rotation {yaw=-12}
   }
   n8 = "Constraint - head_alt (R)"(ComponentPose=Local To Component(LocalPose=Use cached pose 'preEmoteCache'()), Alpha=1, ConstraintWeights_0=1) {
    alphaBoolBlend {blendOption=Linear}
    boneToModify {boneName=head_alt}
    constraintSetup [{transformType=Rotation, targetBone {boneName=head}}]
    constraintWeights [1]
   }
   n9 = Component To Local(ComponentPose=@n8)
   n10 = Sequence Player 'Emote_OrbitTeal_CMM_Visor'!(StartPosition:float=0) -> (Pose) {bLoopAnimation=false, sequence=/BRCosmetics/Animation/Game/MainPlayer/Emotes/OrbitTeal/Emote_OrbitTeal_CMM_Visor}
   Output Pose(Result=Blend Poses by bool(
     BlendPose_0=Layered blend per bone(BasePose=@n9, BlendPoses_0=@n10) {layerSetup [{branchFilters [{boneName=dyn_rangefinder}]}]}
     BlendPose_1=Component To Local(ComponentPose=@n7)
     bActiveValue=Get HeadLimits() -> (HeadLimits:bool)
    ) {blendType=Linear, childUpateMode=ResetChildOnActivate}) {layerGroup=None}
   Save cached pose 'preEmoteCache'(Pose=Copy Pose From Mesh() {bUseAttachedParent=true}) {cacheName=preEmoteCache}
  ]}
 EventGraph {nodes [
   Event BlueprintUpdateAnimation()
   > n1 = Sequence!(execute) -> (then_0, then_1, then_2:Exec, then_3:Exec, then_4:Exec, then_5:Exec, then_6:Exec)
   > n2 = Scale Joint Rotation Function!(
    execute
    Scale:Rotator="1,0.2,1"
    TargetBone:Name=head
    ReferenceBone:Name=neck_02
    SkelMeshComponent:Skinned Mesh Component Object Reference=GetPartSkeletalMeshComponent()
   ) -> (then, FinalRotation:Rotator, OriginalRotation:Rotator)
   > Set head rot(head rot:Rotator=@n2) -> (Output_Get:Rotator)
   > Set head(head:double=GetAlphaFromReferenceTransformDelta(SkeletalMeshComponent=Get Skel_Mesh(), BoneName=head, BaseName=spine_05, InputRange=(X=5,Y=30), TwistAxis="1,0,0")) -> (Output_Get:double)
   n3 = Branch(execute=[@n5.then, @n1.then_1], Condition=Get bIsTargeting())
   > Set isTargetingFloat(isTargetingFloat:double=1) -> (Output_Get:double)
   Delay(execute=@n3.else, Duration=0.2)
   > Set isTargetingFloat(isTargetingFloat=0)
   n4 = GetPartSkeletalMeshComponent()
   n5 = Get Dynamic Material Instance(execute=@n1.then_2, SkeletalMeshComponent=@n4)
   n6 = GetAnimInstance(self=GetOwnerSkeletalMeshComponent())
   Is Valid(exec=@n1.then_3, InputObject=@n6)
   > Set isCustomEmotePlaying_hair(isCustomEmotePlaying_hair:bool=Montage_IsPlaying(self=@n6, Montage=/BRCosmetics/Animation/Game/MainPlayer/Emotes/OrbitTeal/Emote_OrbitTeal_CMM_M)) -> (Output_Get:bool)
   n7 = Branch(
    execute=@n1.then_4
    Condition="float > float"(
     A=GetAlphaFromReferenceTransformDelta(SkeletalMeshComponent=@n4, BoneName=dyn_rangefinder, BaseName=head_alt, DeltaComponentType=EulerY, InputRange=(X=0,Y=90), TwistAxis="1,0,0")
     B=0.8
    )
   )
   > Delay(Duration=0.3)
   > SetScalarParameterValue(self=@n5, ParameterName=emissive mult, Value=5)
   SetScalarParameterValue(execute=@n7.else, self=@n5, ParameterName=emissive mult, Value=0)
   Set isNotPlayingEmote(execute=@n1.then_5, isNotPlayingEmote:bool=NOT Boolean(A=Get bIsPlayingEmote())) -> (Output_Get:bool)
   > Set HeadLimits(HeadLimits:bool=OR Boolean(A=Get isCustomEmotePlaying_hair() -> (isCustomEmotePlaying_hair:bool), B=Get DisableHeadLimits() -> (DisableHeadLimits:bool))) -> (Output_Get:bool)
   Is Valid(exec=@n1.then_6, InputObject=@n6)
   > Set isGuineaPigEmote(isGuineaPigEmote:bool=Montage_IsPlaying(self=@n6, Montage=/BRCosmetics/Animation/Game/MainPlayer/Emotes/GuineaPig/Emote_GuineaPig_CMM_FE_M)) -> (Output_Get:bool)
   Event BlueprintInitializeAnimation()
   n8 = GetOwnerSkeletalMeshComponent()
   > Set Skel_Mesh(Skel_Mesh=@n8)
   > Set Body_Mesh(Body_Mesh=@n8)
   Reroute Node()
   > Is Valid(InputObject=GetAnimInstance(self=GetOwnerSkeletalMeshComponent()))
  ]}
]
variables [
 Body_Mesh {type=SkeletalMeshComponent}
 DisableHeadLimits {type=boolean}
 head {type=number}
 head rot {type=Rotator}
 HeadLimits {type=boolean}
 isCustomEmotePlaying_hair {type=boolean}
 isGuineaPigEmote {type=boolean}
 isNotPlayingEmote {type=boolean}
 isTargetingFloat {type=number}
 Skel_Mesh {type=SkeletalMeshComponent}
]
